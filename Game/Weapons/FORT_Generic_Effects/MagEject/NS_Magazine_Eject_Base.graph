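# Base Niagara system for an ejected weapon magazine, bursting a short-lived mesh that drops with inherited velocity, spin and gravity.
# A custom initialize module reads the mesh renderer, and mesh-offset math adjusts position for the pivot.
# Collision, gated by owner LOD distance and particle age, drives bouncing and rolling before the mesh shrinks away.
# Filtered, rate-limited collision data exports to the User.BPAudioCallback Blueprint handler, presumably for impact audio.
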NiagaraSystem /Game/Weapons/FORT_Generic_Effects/MagEject/NS_Magazine_Eject_Base
dependencies {
 usedDataInterfaces [&NiagaraDataInterfaceMeshRendererInfo, &NiagaraDataInterfaceCurve, /Script/Niagara.NiagaraDataInterfaceExport]
 usedDynamicInputs []
 usedModules [&SystemState]
}
emitters {Magazine {
  emitterUpdateScript [
   &EmitterState()
   &SpawnBurst_Instantaneous(Use Spawn Probability=false, UseLoopCountLimit=false)
  ]
  particleSpawnScript [
   &self:Magazine_Initialize(Mesh Renderer Info=DataInterface() {propertyValues {MeshRenderer=&self:Magazine_0.NiagaraMeshRendererProperties_0}}) {module=Magazine_Initialize}
   &InitialMeshOrientation(Mesh Orientation Mode=::NewEnumerator1, Use Rotation Vector=false)
   &InitializeParticle(
    Color Mode=::NewEnumerator0
    Lifetime=1.25
    Mesh Scale Mode=::NewEnumerator3
    Position Offset=&RotateVectorByQuaternion(Vector To Rotate=$Particles.MeshOffset, Quaternion=$Particles.MeshOrientation)
    UsePositionOffset=true
    Write Material Random=true
   )
   &AddVelocity(
    Velocity=&TransformVector(
     Vector=&RandomRangeVector(Minimum=<0 0 -400>, Maximum=<0 0 -400>)
     Source Space=::World
     Destination Space=::Simulation
    )
    Velocity Speed Scale=0.65
   )
   &InheritVelocity(Inherited Velocity Speed Limit=0, Speed Limit=false, UseSpeedThreshold=false)
   /SetVariables(Particles.SpawnRotationVector=&RandomRangeVector(Minimum=<0.5 -1 0.5>, Maximum=<-0.5 -1 -0.5>)) {isSetParameters=true}
   &AddRotationalVelocity(Rotation Rate=&RandomRangeVector(Minimum=<0.5 -1 0.5>, Maximum=<-0.5 -1 -0.5>)) {enabled=false}
   &ApplyInitialForces()
  ]
  particleUpdateScript [
   &ParticleState()
   &GravityForce()
   &Collision(
    Age Colliding Particles=false
    Collision Enabled=&Boolean_AndOperation(
     Bool1=&SetBoolByFloatComparison(A=$Engine.Owner.LODDistance, B=1500, Comparison Type=::NewEnumerator6)
     Bool2=&SetBoolByFloatComparison(A=$Particles.NormalizedAge, B=0.2)
    )
    Control Roll On Collision=true
    Friction=0.4
    Mesh Dimensions:Vector3f=$StackContext.MeshSize
    Module.Control Rest State Via Surface Normal=false
    Module.EnableMaxCollisionCount=false
    Module.Randomize Collision Normal Vector=true
    Particle Radius Scale=0.8
    Radius Calculation Type=::NewEnumerator1
    Radius Scale for Rotation=10
    Randomize Collision Normal=0.12
    Restitution=0.5
    ScaleRateOfRotation=true
   )
   &UpdateMeshOrientation(
    Rotation Rate=&MakeCustomFloatFromBool(Bool=$Particles.HasCollided, True Float=0, False Float=-1)
    Rotation Vector=$Particles.SpawnRotationVector
   )
   &UpdateMeshOrientation(
    Mesh Radius Calculation Method:ENiagaraCalculateRadiusOptions=::NewEnumerator0
    Mesh Renderer=DataInterface() {propertyValues {MeshRenderer=&self:Magazine_0.NiagaraMeshRendererProperties_0}}
    Orientation Method=::NewEnumerator6
   ) {module=UpdateMeshOrientation001}
   &Drag(Use Linear Drag=true, Use Rotational Drag=true)
   &SolveForcesAndVelocity(Clamp Velocity=false, Limit Acceleration=false, Manually Enable Rotational Solver=true)
   /SetVariables(Particles.PositionWithOffset:NiagaraPosition=&SubtractVectorFromPosition(
     Position=$Particles.Position
     Vector=&RotateVectorByQuaternion(Vector To Rotate=$Particles.MeshOffset, Quaternion=$Particles.MeshOrientation)
    )) {isSetParameters=true}
   &ScaleMeshSize(Scale Factor=&VectorFromFloat(Value=&FloatFromCurve(FloatCurve=DataInterface() {propertyValues {Curve {keys [
          <at 0.02 1 RCIM_Constant>
          {interpMode=RCIM_Cubic, tangentMode=RCTM_User, time=0.9, value=1}
          {arriveTangent=-18.3792, interpMode=RCIM_Cubic, leaveTangent=-18.3792, tangentMode=RCTM_User, time=1, value=0}
         ]}}})))
   /Niagara/Modules/ExportParticleData/ExportParticleDataToBlueprint(
    Condition To Export Data:NiagaraBool=&Boolean_AndOperation(
     Bool1=$Output.Collision.CollisionValid
     Bool2=&Boolean_AndOperation(
      Bool1=$Particles.HasCollided
      Bool2=&SetBoolByIntComparison(A=$Particles.Number Of Collisions, B=1, Comparison Type=::NewEnumerator7)
     )
    )
    Delay Between Data Exports=0.3
    Export Particle Data Interface=DataInterface() {propertyValues {
      GPUAllocationFixedSize=64
      GPUAllocationMode=PerParticle
      GPUAllocationPerParticleSize=1
      CallbackHandlerParameter {parameter {name=User.BPAudioCallback, typeDefHandle {registeredTypeIndex=1}, varData []}}
     }}
    UseDelayTimer=true
   )
  ]
  rendererClasses [&NiagaraMeshRendererProperties]
  renderers [{rendererClass=&NiagaraMeshRendererProperties, rendererIndex=0}]
 }}
summary {
 systemName=NS_Magazine_Eject_Base
 emitters [{emitterName=Magazine, rendererClasses [&NiagaraMeshRendererProperties]}]
 userVariables [
  User.BPAudioCallback {
   type=Object
   description=""
   defaultValue {struct=NiagaraExt_VariableValue_Object, value {object=null, objectClass=&Object}}
  }
  User.StaticMesh {
   type=StaticMesh
   description=""
   defaultValue {
    struct=NiagaraExt_VariableValue_Object
    value {object=&/SM_Magazine_Eject_Base, objectClass=&StaticMesh}
   }
  }
 ]
}
